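; Dance grenade projectile, a B_Prj_Grenade child that applies the GE_DanceStun_UC gameplay effect to pawns caught in its blast.
; Bounce or capsule overlap ends the fuse, triggering a one-time explosion, with traced water splash effects on water entry or detonation.
; On authority, explosion handling skips targets behind tagged obstructors, expands vehicles to passengers, turns neutral AI hostile, sets fall instigators and broadcasts an enemy-hit count for quests.
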(event EventBeginPlay
  (|CallParentFunction))

(event OnComponentBeginOverlap(CapsuleComponent)
  (params (OverlappedComponent :type Primitive%20Component%20Object%20Reference) (OtherActor :type Actor%20Object%20Reference) (OtherComp :type Primitive%20Component%20Object%20Reference) (OtherBodyIndex :type Integer) (bFromSweep :type Boolean) (SweepResult :type Hit%20Result%20Structure%20%28by%20ref%29))
  (bind (_bblockinghit _binitialoverlap _time _distance _location _impactpoint _normal _impactnormal _physmat _hitactor) (Collision|BreakHitResult SweepResult))
  ;[1 1 1] If we overlap a pawn, fake a bounce.
  (bind _aspawn (Utilities|Casting|CastToPawn _hitactor))
  (if (Utilities|NotEqual(Object) _aspawn (Game|GetInstigator))
    (CallFunction|FuseEnded)))

(event EventOnBounce
  (params (Hit :type Hit%20Result%20Structure%20%28by%20ref%29))
  (bind (_bblockinghit _binitialoverlap _time _distance _location _impactpoint _normal _impactnormal _physmat _hitactor) (Collision|BreakHitResult Hit))
  ;[1 1 1] OnBounce - Start fuse
  (|CallParentFunction Hit)
  (bind _aspawn (Utilities|Casting|CastToPawn _hitactor)
    (:then
      (if (Utilities|NotEqual(Object) _aspawn (Game|GetInstigator))
        _))
    (:CastFailed
      _))
  (CallFunction|FuseEnded))

(event OnEnterWaterEvent(WaterInteractionComponent)
  (params (WaterBody :type Fort%20Water%20Body%20Actor%20Object%20Reference) (WaterInteractionComponent :type Fort%20Water%20Interaction%20Component%20Object%20Reference) (bIsFirstBody :type Boolean))
  (bind _mesh (Variables|B_Prj_Grenade|GetMesh))
  ;[1 1 1] Check to see if deploying in water and then Get water info
  (bind (_outhit _returnvalue) (Collision|LineTraceForObjects (+ (Transformation|GetActorLocation) "0, 0,100.000000") (- (Transformation|GetActorLocation) "0, 0,100.000000") (Utilities|Array|MakeArray "ObjectTypeQuery12")))
  (if _returnvalue
    (bind _asfort_water_body_actor (Utilities|Casting|CastToFortWaterBodyActor (.HitActor (Collision|BreakHitResult _outhit))))
    (Variables|Default|SetWaterLocation (.Location (Collision|BreakHitResult _outhit)))
    (bind _outwatersurfaceinfo (WaterBody|GetWaterSurfaceInfoStruct _asfort_water_body_actor (.Location (Collision|BreakHitResult _outhit)) true true false))
    (Rendering|Material|SetScalarParameterValueonMaterials _mesh "WaterDepth" (.WaterDepth (Utilities|Struct|BreakWaterSurfaceInfo _outwatersurfaceinfo)))
    (Rendering|Material|SetScalarParameterValueonMaterials _mesh "WaterIndex" (Math|Conversions|ToFloat(Integer) (.WaterBodyIdx (Utilities|Struct|BreakWaterSurfaceInfo _outwatersurfaceinfo))))
    (bind _returnvalue_2 (Effects|Components|ParticleSystem|SpawnFXAtLocation "/Game/Effects/Niagara/Systems/Enviroment/Water/NS_WaterSplash_Medium.NS_WaterSplash_Medium" (.WaterSurfaceLocation (Utilities|Struct|BreakWaterSurfaceInfo _outwatersurfaceinfo))))
    (Effects|Components|ParticleSystem|SetFloatParameter _returnvalue_2 "User.EffectScale" 1.75)
    (Effects|Components|ParticleSystem|SetFloatParameter _returnvalue_2 "User.WaterDepth" (.WaterDepth (Utilities|Struct|BreakWaterSurfaceInfo _outwatersurfaceinfo)))
    (Effects|Components|ParticleSystem|SetVectorParameter _returnvalue_2 "User.WaterPlaneLocation" (Utilities|Struct|BreakWaterSurfaceInfo _outwatersurfaceinfo))
    (Effects|Components|ParticleSystem|SetVectorParameter _returnvalue_2 "User.WaterPlaneNormal" (.WaterPlaneNormal (Utilities|Struct|BreakWaterSurfaceInfo _outwatersurfaceinfo)))))

(event EventFuseEnded
  (bind _mesh (Variables|B_Prj_Grenade|GetMesh))
  ;[1 1 1] Fuse ended - explode
  (Utilities|FlowControl|DoOnce false
    (:Completed
      (Delivery|DoExplosion)
      ;[1 1 1] Check to see if deploying in water and then Get water info
      (bind (_outhit _returnvalue) (Collision|LineTraceForObjects (+ (Transformation|GetActorLocation) "0, 0,100.000000") (- (Transformation|GetActorLocation) "0, 0,100.000000") (Utilities|Array|MakeArray "ObjectTypeQuery12")))
      (if _returnvalue
        (bind _asfort_water_body_actor (Utilities|Casting|CastToFortWaterBodyActor (.HitActor (Collision|BreakHitResult _outhit))))
        (Variables|Default|SetWaterLocation (.Location (Collision|BreakHitResult _outhit)))
        (bind _outwatersurfaceinfo (WaterBody|GetWaterSurfaceInfoStruct _asfort_water_body_actor (.Location (Collision|BreakHitResult _outhit)) true true false))
        (Rendering|Material|SetScalarParameterValueonMaterials _mesh "WaterDepth" (.WaterDepth (Utilities|Struct|BreakWaterSurfaceInfo _outwatersurfaceinfo)))
        (Rendering|Material|SetScalarParameterValueonMaterials _mesh "WaterIndex" (Math|Conversions|ToFloat(Integer) (.WaterBodyIdx (Utilities|Struct|BreakWaterSurfaceInfo _outwatersurfaceinfo))))
        (bind _returnvalue_2 (Effects|Components|ParticleSystem|SpawnFXAtLocation "/Game/Effects/Niagara/Systems/Enviroment/Water/NS_WaterSplash_Medium.NS_WaterSplash_Medium" (.WaterSurfaceLocation (Utilities|Struct|BreakWaterSurfaceInfo _outwatersurfaceinfo))))
        (Effects|Components|ParticleSystem|SetFloatParameter _returnvalue_2 "User.EffectScale" 1.75)
        (Effects|Components|ParticleSystem|SetFloatParameter _returnvalue_2 "User.WaterDepth" (.WaterDepth (Utilities|Struct|BreakWaterSurfaceInfo _outwatersurfaceinfo)))
        (Effects|Components|ParticleSystem|SetVectorParameter _returnvalue_2 "User.WaterPlaneLocation" (Utilities|Struct|BreakWaterSurfaceInfo _outwatersurfaceinfo))
        (Effects|Components|ParticleSystem|SetVectorParameter _returnvalue_2 "User.WaterPlaneNormal" (.WaterPlaneNormal (Utilities|Struct|BreakWaterSurfaceInfo _outwatersurfaceinfo)))))))

(event EventOnExploded
  (params (HitActors :type Array%20of%20Actor%20Object%20References) (HitResults :type Array%20of%20Hit%20Result%20Structures))
  ;[1 1 1] OnExploded - Emitter, Sound
  (|CallParentFunction HitActors HitResults)
  (Networking|SwitchHasAuthority
    (:Authority
      (for _array_element HitActors
        (bind _asfort_pawn (Utilities|Casting|CastToFortPawn _array_element))
        ;;[1 1 1] BWayland - Fix for effect being applied to players through a shield bubble (and any other desired obstructing actor by adding to ObstructorTags). Should really be handled by ability targeting, but not a good way to maintain the desired functionality that way
        (bind _returnvalue (CallFunction|CheckObstructed _asfort_pawn))
        (if (not _returnvalue)
          (bind _asfort_character_vehicle (Utilities|Casting|CastToFortCharacterVehicle _asfort_pawn)
            (:then
              (for _array_element_2 (Vehicle|GetAllPassengers _asfort_character_vehicle)
                (Variables|Default|SetTarget _array_element_2)
                (GameplayEffects|ApplyGameplayEffectToSelf :self (Ability|GetAbilitySystemComponent (Variables|Default|GetTarget)) :GameplayEffectClass "/Game/Athena/Items/Consumables/DanceGrenade/GE_DanceStun_UC.GE_DanceStun_UC_C" :Level 0.0)
                ;[1 1 1] MakeNeutralAIsHostile
                (bind _asfort_athena_aibot_controller (Utilities|Casting|CastToFortAthenaAIBotController (Pawn|GetController (Variables|Default|GetTarget)))
                  (:then
                    (if (Utilities|Enum|Equal(Enum) (Team|GetTeamAffiliation _asfort_athena_aibot_controller (Variables|Actor|GetInstigator)) "Neutral")
                      (Affiliation|ChangeFactionAttitudeTowardsActor (Fortnite|Factions|GetFactionTagContainerforActor (Variables|Default|GetTarget)) (Variables|Actor|GetInstigator) "Hostile")
                      (else
                        _))
                    (bind _asfort_player_pawn_athena (Utilities|Casting|CastToFortPlayerPawnAthena (Variables|Default|GetTarget)))
                    (Class|FortPlayerPawnAthena|SetFallInstigator :self _asfort_player_pawn_athena :NewFallInstigator (Game|GetInstigatorController))
                    ;[0 0 0] Count players for quest progress if they are not on the same team as the instigator and they are not DBNO (QS2 revamp)
                    (if (and (not (Team|OnSameTeam (Variables|Actor|GetInstigator) (Variables|Default|GetTarget))) (not (Variables|/Script/FortniteGame.FortPawn|GetbIsDBNO (Variables|Default|GetTarget))))
                      (Math|Integer|IncrementInt (Variables|Default|GetCount))))
                  (:CastFailed
                    _))))
            (:CastFailed
              (Variables|Default|SetTarget _asfort_pawn))) ; Fixes FORT-394322 (Eject gunner from mech vehicle)
        ))
      ;[1 1 1] Quest update for hitting an enemy player and reset the count tracking the number of players hit
      (CallFunction|QuestUpdate)
      (Variables|Default|SetCount 0))))

(fn CheckObstructed
  (params (Target :type Fort%20Pawn%20Object%20Reference))
  (bind _outhit (Collision|LineTraceByChannel (Transformation|GetActorLocation) (Transformation|GetActorLocation Target) "TraceTypeQuery6"))
  (if (Collision|BreakHitResult _outhit)
    (bind _returnvalue (GameplayTags|HasAnyMatchingGameplayTags(Message) (.HitActor (Collision|BreakHitResult _outhit)) (Variables|Default|GetObstructorTags)))
    (return _returnvalue)
    (else
      (return false))))

(fn QuestUpdate
  ;[0 0 0] Check if elligible players were found
  (if (> (Variables|Default|GetCount) 0)
    ;[0 0 0] Send the verb update with the count set to the number of players hit
    (CustomVerb|BroadcastCustomVerbMessage (CustomVerb|MakeCustomVerbMessage (Actor|GetOwner) "(TagName=\"CustomVerb.Hit.DanceGrenade\")" (Variables|Default|GetCount)))))

(fn ConstructionScript
  (|CallParentFunction))
