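; Gameplay ability for the Dance Grenade consumable that clears the dance stun from its owner when activated.
; Activation runs the parent class logic, removes owner gameplay effects that grant the Gameplay.Action.Player.DanceStunned tag, then ends the ability immediately.
; The end-ability event has no logic, and any other grenade behavior is not defined in this graph.
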(event Ability|EventActivateAbility
  (|CallParentFunction)
  (Ability|RemoveGameplayEffectFromOwnerWithGrantedTags "(GameplayTags=((TagName=\"Gameplay.Action.Player.DanceStunned\")))")
  (Ability|EndAbility))

(event Ability|EventOnEndAbility
  (params (bWasCancelled :type Boolean)))
